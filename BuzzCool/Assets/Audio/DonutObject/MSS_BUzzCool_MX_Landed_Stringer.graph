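# One-shot stereo MetaSound music stinger for the BuzzCool donut object, named for its landing.
# A single Wave Player starts a shortened "Attack" music wave on play, feeds both output channels, and signals finished when the wave ends.
# It is spatialized through a donut-landing attenuation asset and routed to the POI music sound class.
# A concurrency override allows one instance and prevents new ones while it plays.
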MetaSoundSource /BuzzCool/Assets/Audio/DonutObject/MSS_BUzzCool_MX_Landed_Stringer
inputs [UE.Source.OnPlay {type=Trigger}]
interfaces [UE.OutputFormat.Stereo:1.0, UE.Source.OneShot:1.0, UE.Source:1.1]
outputs [
 "UE.OutputFormat.Stereo.Audio:0" {type=Audio}
 "UE.OutputFormat.Stereo.Audio:1" {type=Audio}
 UE.Source.OneShot.OnFinished {type=Trigger}
]
pages [Default {
  input_defaults {UE.Source.OnPlay=null}
  nodes [WavePlayer_0 = UE.Wave Player:Stereo(Play=^UE.Source.OnPlay) {also {defaults {Wave Asset=/BuzzCool/Assets/Audio/wavs/MX_S38_IP_BuzzCool_Attack_LP_shorten_01}, version [1,1]}}]
  links [
   {dst="@UE.Source.OneShot.OnFinished", src=WavePlayer_0.On Finished}
   {dst="@UE.OutputFormat.Stereo.Audio:0", src=WavePlayer_0.Out Left}
   {dst="@UE.OutputFormat.Stereo.Audio:1", src=WavePlayer_0.Out Right}
  ]
 }]
properties {
 assetImportData=&self:AssetImportData
 attenuationSettings=&/ATT_Donut_Land_Stinger
 bOverrideConcurrency=true
 maxDistance=15000
 numChannels=2
 outputFormat=Stereo
 sampleRate=48000
 soundClassObject=/Game/Sounds/FortSoundClassesAndModes/Fort_Music_POI_PSM
 concurrencyOverrides {maxCount=1, resolutionRule=PreventNew}
}
