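# Looping stereo MetaSound source for the BuzzCool donut's ground ambience, using donut attenuation and an outdoor ambient sound class.
# On play, two drop loops and a sparkle brush loop each get a random start offset, short ADSR fade-in and random pitch shift.
# The enveloped layers feed a stereo mixer, where the brush layer's gain is zero.
# The wave array, Multiplier and attenuation distance inputs are unconnected.
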MetaSoundSource /BuzzCool/Assets/Audio/DonutObject/MSS_BuzzCool_Donut_Ground_Loop
inputs [
 In Array {type=WaveAsset:Array}
 Multiplier {type=Float}
 UE.Attenuation.Distance {type=Float}
 UE.Source.OnPlay {type=Trigger}
]
interfaces [UE.Attenuation:1.0, UE.OutputFormat.Stereo:1.0, UE.Source:1.1]
outputs [
 "UE.OutputFormat.Stereo.Audio:0" {type=Audio}
 "UE.OutputFormat.Stereo.Audio:1" {type=Audio}
]
pages [Default {
  input_defaults {
   Multiplier=4
   UE.Attenuation.Distance=12400.001
   UE.Source.OnPlay=null
   In Array [
    /BuzzCool/Assets/Audio/wavs/Glider_Muertos_Flap_Sparkle_01
    /BuzzCool/Assets/Audio/wavs/Glider_Muertos_Flap_Sparkle_02
    /BuzzCool/Assets/Audio/wavs/Glider_Muertos_Flap_Sparkle_03
   ]
  }
  nodes [
   n1 = UE.RandomTime(Next=^UE.Source.OnPlay) {also {defaults {Max=8, Min=1}}}
   ADSREnvelopeADSREnvelope_2 = ADSR Envelope.ADSR Envelope:Audio(Trigger Attack=@n1.On Next) {also {defaults {Attack Time=0.2, Sustain Level=1}}}
   n2 = UE.RandomFloat(Next=@ADSREnvelopeADSREnvelope_2.On Attack Triggered) {also {defaults {Max=1.1, Min=0.9}}}
   WavePlayer_1 = UE.Wave Player:Stereo(Play=@n2.On Next, Pitch Shift=@n2.Value, Start Time=@n1.Value) {also {defaults {Loop="<MetaSoundBuilderResult.SUCCEEDED: 0>", Wave Asset=/BuzzCool/Assets/Audio/wavs/BuzzCool_Drop_Loop_02}, version [1,1]}}
   n3 = UE.RandomTime(Next=^UE.Source.OnPlay) {also {defaults {Max=8, Min=1}}}
   ADSREnvelopeADSREnvelope_0 = ADSR Envelope.ADSR Envelope:Audio(Trigger Attack=@n3.On Next) {also {defaults {Attack Time=0.2, Sustain Level=1}}}
   n4 = UE.RandomFloat(Next=@ADSREnvelopeADSREnvelope_0.On Attack Triggered) {also {defaults {Max=1.1, Min=0.9}}}
   WavePlayer_2 = UE.Wave Player:Stereo(Play=@n4.On Next, Pitch Shift=@n4.Value, Start Time=@n3.Value) {also {defaults {Loop="<MetaSoundBuilderResult.SUCCEEDED: 0>", Wave Asset=/BuzzCool/Assets/Audio/wavs/BuzzCool_Drop_Loop_01}, version [1,1]}}
   n5 = UE.RandomTime(Next=^UE.Source.OnPlay) {also {defaults {Max=8, Min=1}}}
   ADSREnvelopeADSREnvelope_1 = ADSR Envelope.ADSR Envelope:Audio(Trigger Attack=@n5.On Next) {also {defaults {Attack Time=0.2, Sustain Level=1}}}
   n6 = UE.RandomFloat(Next=@ADSREnvelopeADSREnvelope_1.On Attack Triggered) {also {defaults {Max=1.1, Min=0.9}}}
   WavePlayer_0 = UE.Wave Player:Stereo(Play=@n6.On Next, Pitch Shift=@n6.Value, Start Time=@n5.Value) {also {
     defaults {Loop="<MetaSoundBuilderResult.SUCCEEDED: 0>", Wave Asset=/BuzzCool/Assets/Audio/wavs/AnimalCaretaking_Brush_Loop_Sparkle}
     version [1,1]
    }}
   AudioMixerAudioMixerStereo4_0 = "AudioMixer.Audio Mixer (Stereo, 4)"(
    In 0 L=UE.Multiply:Audio(AdditionalOperands=@ADSREnvelopeADSREnvelope_2.Out Envelope, PrimaryOperand=@WavePlayer_1.Out Left).Out
    In 1 R=UE.Multiply:Audio(PrimaryOperand=@ADSREnvelopeADSREnvelope_0.Out Envelope, AdditionalOperands=@WavePlayer_2.Out Right).Out
    In 3 L=UE.Multiply:Audio(AdditionalOperands=@ADSREnvelopeADSREnvelope_1.Out Envelope, PrimaryOperand=@WavePlayer_0.Out Left).Out
    In 0 R=UE.Multiply:Audio(AdditionalOperands=@ADSREnvelopeADSREnvelope_2.Out Envelope, PrimaryOperand=@WavePlayer_1.Out Right).Out
    In 1 L=UE.Multiply:Audio(PrimaryOperand=@ADSREnvelopeADSREnvelope_0.Out Envelope, AdditionalOperands=@WavePlayer_2.Out Left).Out
    In 3 R=UE.Multiply:Audio(AdditionalOperands=@ADSREnvelopeADSREnvelope_1.Out Envelope, PrimaryOperand=@WavePlayer_0.Out Right).Out
   ) {also {defaults {Gain 2=0.25, Gain 3=0}}}
   InitVariable.Bool()
   InitVariable.Trigger()
  ]
  links [
   {dst="@UE.OutputFormat.Stereo.Audio:0", src=AudioMixerAudioMixerStereo4_0.Out L}
   {dst="@UE.OutputFormat.Stereo.Audio:1", src=AudioMixerAudioMixerStereo4_0.Out R}
  ]
 }]
properties {
 assetImportData=&self:AssetImportData
 attenuationSettings=&/ATT_Donut_Ambient
 bLooping=true
 bOverrideConcurrency=true
 duration=10000
 maxDistance=16000
 numChannels=2
 outputFormat=Stereo
 sampleRate=48000
 soundClassObject=&Fort_Ambient_Outdoors_Athena
 concurrencyOverrides {maxCount=2}
}
