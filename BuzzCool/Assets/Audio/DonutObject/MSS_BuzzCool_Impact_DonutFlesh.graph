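# One-shot stereo MetaSound Source for flesh impact sounds on the BuzzCool donut object.
# On play, it picks a random flesh impact wave, plays it with a random pitch shift, and scales both channels by the Multiplier input.
# Golden, Size and attenuation distance inputs are unconnected.
# It uses pickaxe impact attenuation, the Fort_Impacts class, a concurrency override, a music sidechain bus send and CB_SonicSpear volume modulation.
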MetaSoundSource /BuzzCool/Assets/Audio/DonutObject/MSS_BuzzCool_Impact_DonutFlesh
inputs [
 Flesh Impact {type=WaveAsset:Array}
 Golden {type=Bool}
 Multiplier {type=Float}
 Size {type=Float}
 UE.Attenuation.Distance {type=Float}
 UE.Source.OnPlay {type=Trigger}
]
interfaces [UE.Attenuation:1.0, UE.OutputFormat.Stereo:1.0, UE.Source.OneShot:1.0, UE.Source:1.1]
outputs [
 "UE.OutputFormat.Stereo.Audio:0" {type=Audio}
 "UE.OutputFormat.Stereo.Audio:1" {type=Audio}
 UE.Source.OneShot.OnFinished {type=Trigger}
]
pages [Default {
  input_defaults {
   Golden=null
   Multiplier=3.8
   Size=0
   UE.Attenuation.Distance=0
   UE.Source.OnPlay=null
   Flesh Impact [
    /BuzzCool/Assets/Audio/wavs/BuzzCool_Impact_Flesh_01
    /BuzzCool/Assets/Audio/wavs/BuzzCool_Impact_Flesh_02
    /BuzzCool/Assets/Audio/wavs/BuzzCool_Impact_Flesh_03
    /BuzzCool/Assets/Audio/wavs/BuzzCool_Impact_Flesh_04
    /BuzzCool/Assets/Audio/wavs/BuzzCool_Impact_Flesh_05
    /BuzzCool/Assets/Audio/wavs/BuzzCool_Impact_Flesh_06
   ]
  }
  nodes [
   ArrayRandomGet_0 = Array.Random Get:WaveAsset:Array(In Array=^Flesh Impact, Next=^UE.Source.OnPlay) {also {defaults {Weights []}}}
   n1 = UE.RandomFloat(Next=@ArrayRandomGet_0.On Next) {also {defaults {Max=1.2, Min=0.9}}}
   WavePlayer_0 = UE.Wave Player:Stereo(Wave Asset=@ArrayRandomGet_0.Value, Play=@n1.On Next, Pitch Shift=@n1.Value) {also {version [1,1]}}
   Multiply_0 = UE.Multiply:Audio by Float(AdditionalOperands=^Multiplier, PrimaryOperand=@WavePlayer_0.Out Right)
   Multiply_1 = UE.Multiply:Audio by Float(AdditionalOperands=^Multiplier, PrimaryOperand=@WavePlayer_0.Out Left)
  ]
  links [
   {dst="@UE.OutputFormat.Stereo.Audio:1", src=Multiply_0.Out}
   {dst="@UE.OutputFormat.Stereo.Audio:0", src=Multiply_1.Out}
   {dst="@UE.Source.OneShot.OnFinished", src=WavePlayer_0.On Finished}
  ]
 }]
properties {
 assetImportData=&self:AssetImportData
 attenuationSettings=/Game/Sounds/Attenuation/Fort_Weapon_Pickaxe_Player_Impact_Att
 bOverrideConcurrency=true
 maxDistance=3512
 numChannels=2
 outputFormat=Stereo
 sampleRate=48000
 soundClassObject=&Fort_Impacts
 busSends [{
   audioBus=/Game/Sounds/Modulation/EnvelopeFollowers/HPF/AB_SideChain_FortMusic_HPF
   minSendDistance=300
   sendLevel=0.8
   sourceBusSendLevelControlMethod=Linear
  }]
 concurrencyOverrides {maxCount=5, retriggerTime=0.1, volumeScale=0.6}
 modulationSettings {volumeRouting=Override, volumeModulationDestination {bEnableModulation=true, modulators [&CB_SonicSpear]}}
}
