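# Sound cue for the BuzzCool donut explosion, randomly playing one of three SonicSpear death waves borrowed from the SonicSpearGameplay content.
# An equally weighted Random node feeds a Modulator at the root.
# It uses the BuzzCool explode attenuation and the Fort NPC explosions sound class.
# Concurrency overrides cap simultaneous instances at four and prevent new ones beyond that.
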SoundCue /BuzzCool/Assets/Audio/DonutObject/SC_BuzzCool_Donut_Explode
root=Modulator_0
nodes [
 {id=Modulator_0, children [Random_2]}
 {id=Random_2, children [WavePlayer_9, WavePlayer_13, WavePlayer_14], properties {weights [1,1,1]}}
 {id=WavePlayer_13, properties {soundWaveAssetPtr=/SonicSpearGameplay/Assets/Audio/wav/SonicSpear_Death_02}}
 {id=WavePlayer_14, properties {soundWaveAssetPtr=/SonicSpearGameplay/Assets/Audio/wav/SonicSpear_Death_03}}
 {id=WavePlayer_9, properties {soundWaveAssetPtr=/SonicSpearGameplay/Assets/Audio/wav/SonicSpear_Death_01}}
]
properties {
 attenuationSettings=/BuzzCool/Assets/Audio/Clones/Settings/ATT_BuzzCool_Explode
 maxDistance=16512
 soundClassObject=/Game/Sounds/FortSoundClassesAndModes/Fort_Events_Explosions_NPC
 volumeMultiplier=0.85
 concurrencyOverrides {maxCount=4, resolutionRule=PreventNew}
}
