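# Sound cue for the voice layer of the BuzzCool donut explosion, using the explosion attenuation and NPC explosion sound class.
# A mixer layers a random SonicSpear death sound, boosted, with a quieter random Lager rift-spawn sound that follows after a short randomized delay.
# The mix runs through a default Modulator for variation, and concurrency overrides set a small instance cap that prevents new plays.
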SoundCue /BuzzCool/Assets/Audio/DonutObject/SC_BuzzCool_Donut_Explode_Voices
root=Modulator_0
nodes [
 {id=Delay_0, children [Random_0], properties {delayMax=0.4, delayMin=0.3}}
 {id=Mixer_0, children [Random_2,Delay_0], properties {inputVolume [1.5,0.25]}}
 {id=Modulator_0, children [Mixer_0]}
 {id=Random_0, children [WavePlayer_0,WavePlayer_1,WavePlayer_2,WavePlayer_5,WavePlayer_6,WavePlayer_7], properties {weights [1,1,1,1,1,1]}}
 {id=Random_2, children [WavePlayer_9, WavePlayer_13, WavePlayer_14], properties {weights [1,1,1]}}
 {id=WavePlayer_0, properties {soundWaveAssetPtr=/BuzzCool/Assets/Audio/wavs/Lager_Spawn_Rift_03}}
 {id=WavePlayer_1, properties {soundWaveAssetPtr=/BuzzCool/Assets/Audio/wavs/Lager_Spawn_Rift_04}}
 {id=WavePlayer_13, properties {soundWaveAssetPtr=/SonicSpearGameplay/Assets/Audio/wav/SonicSpear_Death_02}}
 {id=WavePlayer_14, properties {soundWaveAssetPtr=/SonicSpearGameplay/Assets/Audio/wav/SonicSpear_Death_03}}
 {id=WavePlayer_2, properties {soundWaveAssetPtr=/BuzzCool/Assets/Audio/wavs/Lager_Spawn_Rift_05}}
 {id=WavePlayer_5, properties {soundWaveAssetPtr=/BuzzCool/Assets/Audio/wavs/Lager_Spawn_Rift_06}}
 {id=WavePlayer_6, properties {soundWaveAssetPtr=/BuzzCool/Assets/Audio/wavs/Lager_Spawn_Rift_07}}
 {id=WavePlayer_7, properties {soundWaveAssetPtr=/BuzzCool/Assets/Audio/wavs/Lager_Spawn_Rift_08}}
 {id=WavePlayer_9, properties {soundWaveAssetPtr=/SonicSpearGameplay/Assets/Audio/wav/SonicSpear_Death_01}}
]
properties {
 attenuationSettings=/BuzzCool/Assets/Audio/Clones/Settings/ATT_BuzzCool_Explode
 maxDistance=16512
 soundClassObject=/Game/Sounds/FortSoundClassesAndModes/Fort_Events_Explosions_NPC
 volumeMultiplier=1
 concurrencyOverrides {maxCount=4, resolutionRule=PreventNew}
}
