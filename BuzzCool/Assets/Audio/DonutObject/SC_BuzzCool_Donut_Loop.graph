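# Looping ambient Sound Cue for the BuzzCool donut object, with its own attenuation and concurrency limited to one instance, rejecting new ones.
# A Mixer blends three looping beds (giant loop, sky rift energy, rift spawner ambience) with two looping random one-shot layers.
# Those layers fire sparkle flaps and distant rift bursts after random delays.
# An Enveloper fades the mix in, and a Modulator at the root varies pitch slightly.
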SoundCue /BuzzCool/Assets/Audio/DonutObject/SC_BuzzCool_Donut_Loop
root=Modulator_1
nodes [
 {id=Delay_0, children [Random_0], properties {delayMax=3, delayMin=1}}
 {id=Delay_2, children [Random_3], properties {delayMax=3, delayMin=1}}
 {
  id=Enveloper_0
  children [Mixer_1]
  properties {pitchCurve {editorCurveData {keys [{time=0}]}}, volumeCurve {editorCurveData {keys [<at 0 0>,<at 1 1>]}}}
 }
 {id=Looping_0, children [Delay_0]}
 {id=Looping_2, children [Delay_2]}
 {id=Mixer_1, children [WavePlayer_7, WavePlayer_21, WavePlayer_14, Looping_0, Looping_2], properties {inputVolume [0.8,0.25,0.25,0.5,0.85]}}
 {id=Modulator_1, children [Enveloper_0], properties {pitchMax=1.1, pitchMin=0.9}}
 {id=Random_0, children [WavePlayer_3,WavePlayer_4,WavePlayer_5], properties {weights [1,1,1]}}
 {
  id=Random_3
  children [WavePlayer_22, WavePlayer_23, WavePlayer_24, WavePlayer_25, WavePlayer_26, WavePlayer_27, WavePlayer_28, WavePlayer_29]
  properties {weights [1,1,1,1,1,1,1,1]}
 }
 {id=WavePlayer_14, properties {bLooping=true, soundWaveAssetPtr=/BuzzCool/Assets/Audio/wavs/Juno_RiftSpawner_Ambient_Loop_02}}
 {id=WavePlayer_21, properties {bLooping=true, soundWaveAssetPtr=/BuzzCool/Assets/Audio/wavs/LodeSwan_Island_SkyRift_Energy_Loop}}
 {id=WavePlayer_22, properties {soundWaveAssetPtr=/BuzzCool/Assets/Audio/wavs/RiftEncounter_Open_Burst_Distant_01}}
 {id=WavePlayer_23, properties {soundWaveAssetPtr=/BuzzCool/Assets/Audio/wavs/RiftEncounter_Open_Burst_Distant_02}}
 {id=WavePlayer_24, properties {soundWaveAssetPtr=/BuzzCool/Assets/Audio/wavs/RiftEncounter_Open_Burst_Distant_03}}
 {id=WavePlayer_25, properties {soundWaveAssetPtr=/BuzzCool/Assets/Audio/wavs/RiftEncounter_Open_Burst_Distant_04}}
 {id=WavePlayer_26, properties {soundWaveAssetPtr=/BuzzCool/Assets/Audio/wavs/RiftEncounter_Open_Burst_Distant_05}}
 {id=WavePlayer_27, properties {soundWaveAssetPtr=/BuzzCool/Assets/Audio/wavs/RiftEncounter_Open_Burst_Distant_06}}
 {id=WavePlayer_28, properties {soundWaveAssetPtr=/BuzzCool/Assets/Audio/wavs/RiftEncounter_Open_Burst_Distant_07}}
 {id=WavePlayer_29, properties {soundWaveAssetPtr=/BuzzCool/Assets/Audio/wavs/RiftEncounter_Open_Burst_Distant_08}}
 {id=WavePlayer_3, properties {soundWaveAssetPtr=/BuzzCool/Assets/Audio/wavs/Glider_Muertos_Flap_Sparkle_01}}
 {id=WavePlayer_4, properties {soundWaveAssetPtr=/BuzzCool/Assets/Audio/wavs/Glider_Muertos_Flap_Sparkle_02}}
 {id=WavePlayer_5, properties {soundWaveAssetPtr=/BuzzCool/Assets/Audio/wavs/Glider_Muertos_Flap_Sparkle_03}}
 {id=WavePlayer_7, properties {bLooping=true, soundWaveAssetPtr=/BuzzCool/Assets/Audio/wavs/BuzzCool_Giant_Loop_01}}
]
properties {
 attenuationSettings=&/ATT_Donut_Ambient
 bOverrideConcurrency=true
 maxDistance=16000
 soundClassObject=&Fort_Ambient_Outdoors_Athena
 concurrencyOverrides {maxCount=1, resolutionRule=PreventNew}
}
